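# Apollo landscape material for the non-RVT terrain, ending in the Curie landscape function, with opacity mask from landscape visibility.
# Default and deferred paths blend common layers with slope cliffs and output world-space normals; forward and mobile use simplified mobile layers.
# The mobile path adds a zero-multiplied layer blend to base color, contributing nothing visually.
# Standard grass spawns where slope, Gravel, Sand, Mud and Asphalt masks and landscape holes do not exclude it.
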Material /Game/Athena/Apollo/Environments/Landscape/Materials/NonRVT/M_Apollo_Terrain_Basic
graph {
 properties {bTangentSpaceNormal=false, bUseMaterialAttributes=true, pixelDepthOffsetMode=PDOM_Legacy, usedShadingModels=Default Lit}
 nodes [
  n1 = Constant() {r=2}
  n2 = Max(
   A=Max(A=LandscapeLayerSample() {parameterName=Gravel}, B=LandscapeLayerSample() {parameterName=Sand})
   B=LandscapeLayerSample() {parameterName=Mud}
  )
  n3 = LandscapeLayerSample() {parameterName=Asphalt}
  n4 = Saturate(None=Max(A=Multiply(A=@n1, B=@n2), B=Ceil(None=@n3)))
  n5 = MaterialFunctionCall!(
   AdditionalGrassBorderColorMask
   PutSandBeforeGravel=NamedRerouteDeclaration(None=StaticBool()) {name=PutSandBeforeGravel, nodeColor=<rgba 0.5 0 1 1>}
   UseOceanLevel=NamedRerouteDeclaration(None=StaticBool()) {name=UseOceanLevel, nodeColor=<rgba 0.5 0 1 1>}
  ) {materialFunction=/Game/Athena/Apollo/Environments/Landscape/Materials/Biome/CommonMaterials/MF_CommonLandscapeLayers}
  n6 = MaterialFunctionCall!(
   Material=@n5.Material
   Heightmap=@n5.Material
   NewSlopeMaskMask
   Displacement
   AsphaltLayer
   UseAlternateCliffMat
   AltCliffHeight
   AltCliffMat
  ) {materialFunction=/Game/Athena/Apollo/Environments/Landscape/Materials/Biome/CommonMaterials/MF_SlopeBlendCliff}
  MaterialExpressionMultiply_1 = Multiply(A=OneMinus(None=Max(A=@n6.LandscapeGrassMeshMask, B=@n4)), B=LandscapeVisibilityMask())
  LandscapeGrassOutput!(Grass=@MaterialExpressionMultiply_1) {grassTypes [Grass {
     grassType=/Game/Athena/Apollo/Environments/Landscape/SmallFoliage/DataAssets/LandscapeGrassType/NonRVT/LGT_StandardGrass_NoRVT
     input {
      expression=&self:MaterialExpressionMultiply_1
      inputName=None
      mask=0
      maskA=0
      maskB=0
      maskG=0
      maskR=0
      outputIndex=0
     }
    }]}
  n7 = GetMaterialAttributes(None=MaterialFunctionCall!(PutSandBeforeGravelOverride=StaticBool() {value=true}, UseSnow=StaticBool(), UseOceanLevel=NamedRerouteUsage()) {materialFunction=/Game/Athena/Apollo/Environments/Landscape/Materials/Biome/CommonMaterials/Mobile/MF_Mobile_CommonLandscapeLayers}.Material) {attributeGetTypes [#1]}
  MaterialExpressionConstant_13 = Constant()
  n8 = LandscapeLayerBlend!(
   Layer Gravel=@MaterialExpressionConstant_13
   Layer Mud=@MaterialExpressionConstant_13
   Layer Sand=@MaterialExpressionConstant_13
   Layer Asphalt=@MaterialExpressionConstant_13
  ) {layers [
    {
     blendType=LB_WeightBlend
     constHeightInput=0
     layerName=Gravel
     previewWeight=0
     constLayerInput=<0 0 0>
     heightInput {expression=null, inputName=None, mask=0, maskA=0, maskB=0, maskG=0, maskR=0, outputIndex=0}
     layerInput {
      expression=&self:MaterialExpressionConstant_13
      inputName=None
      mask=0
      maskA=0
      maskB=0
      maskG=0
      maskR=0
      outputIndex=0
     }
    }
    {
     blendType=LB_WeightBlend
     constHeightInput=0
     layerName=Mud
     previewWeight=0
     constLayerInput=<0 0 0>
     heightInput {expression=null, inputName=None, mask=0, maskA=0, maskB=0, maskG=0, maskR=0, outputIndex=0}
     layerInput {
      expression=&self:MaterialExpressionConstant_13
      inputName=None
      mask=0
      maskA=0
      maskB=0
      maskG=0
      maskR=0
      outputIndex=0
     }
    }
    {
     blendType=LB_WeightBlend
     constHeightInput=0
     layerName=Sand
     previewWeight=0
     constLayerInput=<0 0 0>
     heightInput {expression=null, inputName=None, mask=0, maskA=0, maskB=0, maskG=0, maskR=0, outputIndex=0}
     layerInput {
      expression=&self:MaterialExpressionConstant_13
      inputName=None
      mask=0
      maskA=0
      maskB=0
      maskG=0
      maskR=0
      outputIndex=0
     }
    }
    {
     blendType=LB_WeightBlend
     constHeightInput=0
     layerName=Asphalt
     previewWeight=0
     constLayerInput=<0 0 0>
     heightInput {expression=null, inputName=None, mask=0, maskA=0, maskB=0, maskG=0, maskR=0, outputIndex=0}
     layerInput {
      expression=&self:MaterialExpressionConstant_13
      inputName=None
      mask=0
      maskA=0
      maskB=0
      maskG=0
      maskR=0
      outputIndex=0
     }
    }
   ]}
  n9 = SetMaterialAttributes!(MaterialAttributes=@n7.MaterialAttributes, Base Color=Add(A=@n7.Base Color, B=Multiply(A=@n8) {constB=0})) {attributeSetTypes [#1]}
  n10 = MaterialFunctionCall!(Material=@n6.Material) {materialFunction=/Game/Athena/Apollo/Maps/Landscape/VT/MaterialFunctions/TransformTangentToWorld}
  MaterialFunctionCall!(
   In=SetMaterialAttributes!(
    MaterialAttributes=ShadingPathSwitch(Default=@n10.Result, Deferred=@n10.Result, Forward=@n9, Mobile=@n9)
    Opacity Mask=LandscapeVisibilityMask()
   ) {attributeSetTypes [#2]}
   UseHeightFalloff=StaticBool() {desc=Do Not UseHeightFalloff here}
   OnFireEmissiveOverride
   HeightFalloffRadius
   HeightFalloffHardness
   VariationMasksMobile
   In_Grass
   Is_Stylised_WE
  ) {desc=This should be near very last, materialFunction=/Game/Packages/Fortress_SharedMaterials/Base_Material_Functions/MF_Curie_Landscape}
  NamedRerouteDeclaration(None=@n6.HeightmapForGrass) {name=HeightmapForGrass, nodeColor=<rgba 0.5 0 1 1>}
  NamedRerouteUsage()
 ]
}
